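; Looping gameplay cue for the PlacidMelon weapon's beam, handling its Niagara effect and fire audio.
; Loop start and application snap the first particle component to the mask mesh's FX_Beam socket, aim it at the cue location and set LocalView; only loop start sets the beam end.
; Loop start also plays 1P/3P fire sound, plus a whoosh at the beam segment's closest point to the observing player.
; Tick only performs validity checks.
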(event EventTick
  (params (DeltaSeconds :type Float%20%28single-precision%29))
  (Utilities|IsValid (stale EventOnLoopingStart(Niagara) MyTarget)
    (:Is%20Valid
      (Utilities|IsValid (Utilities|Array|Get(acopy) (stale EventOnLoopingStart(Niagara) ParticleComponents))))))

(event EventOnLoopingStart(Niagara)
  (params (MyTarget :type Actor%20Object%20Reference) (Parameters_NormalizedMagnitude :type Float%20%28single-precision%29) (Parameters_RawMagnitude :type Float%20%28single-precision%29) (Parameters_EffectContext :type Gameplay%20Effect%20Context%20Handle%20Structure) (Parameters_MatchedTagName :type Gameplay%20Tag%20Structure) (Parameters_OriginalTag :type Gameplay%20Tag%20Structure) (Parameters_AggregatedSourceTags :type Gameplay%20Tag%20Container%20Structure) (Parameters_AggregatedTargetTags :type Gameplay%20Tag%20Container%20Structure) (Parameters_Location :type Vector :default "0, 0, 0") (Parameters_Normal :type Vector :default "0, 0, 0") (Parameters_Instigator :type Actor%20Object%20Reference) (Parameters_EffectCauser :type Actor%20Object%20Reference) (Parameters_SourceObject :type Object%20Reference) (Parameters_PhysicalMaterial :type Physical%20Material%20Object%20Reference) (Parameters_GameplayEffectLevel :type Integer) (Parameters_AbilityLevel :type Integer) (Parameters_TargetAttachComponent :type Scene%20Component%20Object%20Reference) (Parameters_bReplicateLocationWhenUsingMinimalRepProxy :type Boolean) (ParticleComponents :type Array%20of%20Niagara%20Particle%20System%20Component%20Object%20References) (AudioComponents :type Array%20of%20Audio%20Component%20Object%20References))
  (bind _output (Utilities|Array|Get(acopy) ParticleComponents))
  (bind _asplayer_pawn_athena (Utilities|Casting|CastToPlayerPawn_Athena MyTarget))
  (bind _output_get (Variables|Default|SetPlayer%20Pawn%20Athena _asplayer_pawn_athena))
  (Utilities|FlowControl|Sequence
    (:then_0
      (bind _asb_weap_placid_melon (Utilities|Casting|CastToB_Weap_PlacidMelon (Weapon|GetCurrentWeapon _output_get)))
      (Utilities|IsValid (Variables|/PlacidMelon/Gameplay/PlacidMelon/B_Weap_PlacidMelon.B_Weap_PlacidMelon_C|GetMesh_Mask _asb_weap_placid_melon)
        (:Is%20Valid
          (Transformation|SetWorldLocation _output (Transformation|GetSocketLocation (Variables|/PlacidMelon/Gameplay/PlacidMelon/B_Weap_PlacidMelon.B_Weap_PlacidMelon_C|GetMesh_Mask _asb_weap_placid_melon) "FX_Beam") false false)
          (Transformation|AttachComponentToComponent _output (Variables|/PlacidMelon/Gameplay/PlacidMelon/B_Weap_PlacidMelon.B_Weap_PlacidMelon_C|GetMesh_Mask _asb_weap_placid_melon) "FX_Beam" "KeepRelative" "KeepRelative" "KeepRelative" true)
          (Transformation|SetWorldRotation _output (Math|Conversions|RotationFromXVector (Math|Vector|Normalize (- Parameters_Location (Transformation|GetSocketLocation (Variables|/PlacidMelon/Gameplay/PlacidMelon/B_Weap_PlacidMelon.B_Weap_PlacidMelon_C|GetMesh_Mask _asb_weap_placid_melon) "FX_Beam")))) false false)
          (Effects|Components|ParticleSystem|SetVectorParameter _output "BeamEnd_0" Parameters_Location))))
    (:then_1
      (CallFunction|PlayFireAudio))
    (:then_2
      (Effects|Components|ParticleSystem|SetBoolParameter _output "LocalView" (Pawn|IsLocallyControlled _asplayer_pawn_athena)))))

(event EventOnApplication(EitherVFX-preferthisone)
  (params (MyTarget :type Actor%20Object%20Reference) (Parameters_NormalizedMagnitude :type Float%20%28single-precision%29) (Parameters_RawMagnitude :type Float%20%28single-precision%29) (Parameters_EffectContext :type Gameplay%20Effect%20Context%20Handle%20Structure) (Parameters_MatchedTagName :type Gameplay%20Tag%20Structure) (Parameters_OriginalTag :type Gameplay%20Tag%20Structure) (Parameters_AggregatedSourceTags :type Gameplay%20Tag%20Container%20Structure) (Parameters_AggregatedTargetTags :type Gameplay%20Tag%20Container%20Structure) (Parameters_Location :type Vector :default "0, 0, 0") (Parameters_Normal :type Vector :default "0, 0, 0") (Parameters_Instigator :type Actor%20Object%20Reference) (Parameters_EffectCauser :type Actor%20Object%20Reference) (Parameters_SourceObject :type Object%20Reference) (Parameters_PhysicalMaterial :type Physical%20Material%20Object%20Reference) (Parameters_GameplayEffectLevel :type Integer) (Parameters_AbilityLevel :type Integer) (Parameters_TargetAttachComponent :type Scene%20Component%20Object%20Reference) (Parameters_bReplicateLocationWhenUsingMinimalRepProxy :type Boolean) (ParticleComponents :type Array%20of%20FXSystem%20Component%20Object%20References) (AudioComponents :type Array%20of%20Audio%20Component%20Object%20References) (BurstCameraShakeInstance :type Camera%20Shake%20Base%20Object%20Reference) (BurstDecalInstance :type Decal%20Actor%20Object%20Reference))
  (bind _output (Utilities|Array|Get(acopy) ParticleComponents))
  (bind _asplayer_pawn_athena (Utilities|Casting|CastToPlayerPawn_Athena MyTarget))
  (bind _output_get (Variables|Default|SetPlayer%20Pawn%20Athena _asplayer_pawn_athena))
  (Utilities|FlowControl|Sequence
    (:then_0
      (bind _asb_weap_placid_melon (Utilities|Casting|CastToB_Weap_PlacidMelon (Weapon|GetCurrentWeapon _output_get)))
      (Utilities|IsValid (Variables|/PlacidMelon/Gameplay/PlacidMelon/B_Weap_PlacidMelon.B_Weap_PlacidMelon_C|GetMesh_Mask _asb_weap_placid_melon)
        (:Is%20Valid
          (Transformation|SetWorldLocation _output (Transformation|GetSocketLocation (Variables|/PlacidMelon/Gameplay/PlacidMelon/B_Weap_PlacidMelon.B_Weap_PlacidMelon_C|GetMesh_Mask _asb_weap_placid_melon) "FX_Beam") false false)
          (Transformation|AttachComponentToComponent _output (Variables|/PlacidMelon/Gameplay/PlacidMelon/B_Weap_PlacidMelon.B_Weap_PlacidMelon_C|GetMesh_Mask _asb_weap_placid_melon) "FX_Beam" "KeepWorld" "KeepRelative" "KeepRelative" true)
          (Transformation|SetWorldRotation _output (Math|Conversions|RotationFromXVector (Math|Vector|Normalize (- Parameters_Location (Transformation|GetSocketLocation (Variables|/PlacidMelon/Gameplay/PlacidMelon/B_Weap_PlacidMelon.B_Weap_PlacidMelon_C|GetMesh_Mask _asb_weap_placid_melon) "FX_Beam")))) false false))))
    (:then_1
      (Effects|Components|ParticleSystem|SetBoolParameter _output "LocalView" (Pawn|IsLocallyControlled _asplayer_pawn_athena)))))

(fn GetClosestPointonLasertoObserver
  (bind _self self)
  (bind _returnvalue (Game|GetPlayerPawn _self))
  (bind _player_pawn_athena (Variables|Default|GetPlayer%20Pawn%20Athena))
  (Utilities|IsValid _player_pawn_athena
    (:Is%20Valid
      (Utilities|IsValid _returnvalue
        (:Is%20Valid
          (return (Math|Vector|FindClosestPointonSegment (Transformation|GetActorLocation _returnvalue) (Transformation|GetSocketLocation (Variables|/Script/Engine.Character|GetMesh _player_pawn_athena) "FX_Head") (+ (Transformation|GetActorLocation) (* (Math|Vector|GetForwardVector (Variables|/Game/Athena/PlayerPawn_Athena.PlayerPawn_Athena_C|GetPlayerAimRotation _player_pawn_athena)) (ScalableFloat|GetValueatLevel (Variables|Default|GetRange)))))))))))

(fn PlayFireAudio
  (Utilities|FlowControl|Sequence
    (:then_0
      ;[1 1 1] Play fire sound
      (bind _player_pawn_athena (Variables|Default|GetPlayer%20Pawn%20Athena)
        (:then
          (Audio|SpawnSoundAttached (select (Pawn|IsLocallyControlled _player_pawn_athena) (Variables|Default|GetFire%20Sound%201P) (Variables|Default|GetFire%20Sound%203P)) (Variables|/Script/Engine.Actor|GetRootComponent (Variables|Default|GetPlayer%20Pawn%20Athena))))
        (:else)))
    (:then_1
      (Utilities|FlowControl|Sequence
        (:then_0
          ;[1 1 1] Play whoosh by sound
          (bind _player_pawn_athena_2 (Variables|Default|GetPlayer%20Pawn%20Athena)
            (:then
              (Audio|SpawnSoundatLocation _player_pawn_athena_2 (Variables|Default|GetWhoosh%20By%20Sound) (CallFunction|GetClosestPointOnLaserToObserver)))
            (:else)))))))

(fn ConstructionScript)
