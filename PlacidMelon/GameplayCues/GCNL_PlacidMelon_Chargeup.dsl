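; Looping gameplay cue notify that handles the charge-up Niagara effect for the PlacidMelon weapon.
; On looping start, it casts the cue target to a player pawn and its current weapon to B_Weap_PlacidMelon.
; If the weapon's Mask mesh is valid, it attaches the first particle component to that mesh's FX_Beam socket, keeping relative transforms.
; OnBurst just returns false, and the construction script is empty.
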(event EventOnLoopingStart(Niagara)
  (params (MyTarget :type Actor%20Object%20Reference) (Parameters :type Gameplay%20Cue%20Parameters%20Structure%20%28by%20ref%29) (ParticleComponents :type Array%20of%20Niagara%20Particle%20System%20Component%20Object%20References) (AudioComponents :type Array%20of%20Audio%20Component%20Object%20References))
  (bind _asplayer_pawn_athena (Utilities|Casting|CastToPlayerPawn_Athena MyTarget))
  (bind _asb_weap_placid_melon (Utilities|Casting|CastToB_Weap_PlacidMelon (Weapon|GetCurrentWeapon _asplayer_pawn_athena)))
  (Utilities|IsValid (Variables|/PlacidMelon/Gameplay/PlacidMelon/B_Weap_PlacidMelon.B_Weap_PlacidMelon_C|GetMesh_Mask _asb_weap_placid_melon)
    (:Is%20Valid
      (Transformation|AttachComponentToComponent (Utilities|Array|Get(acopy) ParticleComponents) (Variables|/PlacidMelon/Gameplay/PlacidMelon/B_Weap_PlacidMelon.B_Weap_PlacidMelon_C|GetMesh_Mask _asb_weap_placid_melon) "FX_Beam" "KeepRelative" "KeepRelative" "KeepRelative" true))))

(fn OnBurst
  (params (MyTarget :type Actor%20Object%20Reference) (Parameters :type Gameplay%20Cue%20Parameters%20Structure%20%28by%20ref%29))
  (return false))

(fn ConstructionScript)
